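# Character cosmetic Niagara effect that places music-reactive fire on the NoiseClueGust body's sockets.
# Persistent mesh flames follow sockets, feeding smoothed positions to InterpolateDynamic and brightening with audio; sprite decals add glow.
# Music amplitude, or a fake front-end loop, gates extra flame puffs and curl-noise embers.
# HideFX suppresses spawning, while partition arrays map particles to sockets and let hand/foot-hidden flags drop matching particles.
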NiagaraSystem /BRCosmetics/Characters/Player/Male/Medium/Bodies/M_MED_NoiseClueGust/FX/NS_Character_NoiseClueGust_FX
dependencies {
 usedDataInterfaces [
  &NiagaraDataInterfaceArrayInt32
  &NiagaraDataInterfaceCurve
  &NiagaraDataInterfaceArrayFloat3
  &NiagaraDataInterfaceArrayFloat
  &NiagaraDataInterfaceArrayFloat2
  &NiagaraDataInterfaceSocketReader
  &NiagaraDataInterfaceColorCurve
 ]
 usedDynamicInputs [
  &VectorLength
  &ColorFromCurve
  &RemapRange
  &Waveform
  &MakeFloatFromLinearColor
 ]
 usedModules [
  &self:SystemSpawnScript.SystemScriptSource.SystemScriptGraph.NiagaraNodeAssignment_1.SetVariables_#1
  &self:SystemSpawnScript.SystemScriptSource.SystemScriptGraph.NiagaraNodeAssignment_9.SetVariables_#2
  &SystemState
  &self:SystemSpawnScript.SystemScriptSource.SystemScriptGraph.NiagaraNodeAssignment_0.SetVariables_#3
  &self:SystemSpawnScript.SystemScriptSource.SystemScriptGraph.NiagaraNodeAssignment_2.SetVariables_#4
  &self:SystemSpawnScript.SystemScriptSource.SystemScriptGraph.NiagaraNodeAssignment_4.SetVariables_#5
  &BlendSpaceHelper
  /Game/Effects/Niagara/Modules/Audio/FN_FakeAudioReactivityLoop
  &self:SystemSpawnScript.SystemScriptSource.SystemScriptGraph.NiagaraNodeAssignment_20.SetVariables_#6
  &self:SystemSpawnScript.SystemScriptSource.SystemScriptGraph.NiagaraNodeAssignment_21.SetVariables_#7
  &self:SystemSpawnScript.SystemScriptSource.SystemScriptGraph.NiagaraNodeAssignment_22.SetVariables_#8
  &Timeline
 ]
}
emitters {
 AdditionalFlames {
  emitterUpdateScript [
   &EmitterState(
    Inactive Response=::NewEnumerator1
    Life Cycle Mode=::NewEnumerator1
    Loop Behavior=::NewEnumerator1
    Loop Duration Mode=::NewEnumerator1
    UseLoopDelay=false
   )
   &SpawnRate(
    Spawn Probability=&MakeCustomFloatFromBool(
     Bool=&SetBoolByFloatComparison(
      A=&FN_FrontEndFloatModifier(Gameplay=$NPC.NPC_Audio_Reactive_Parameters.Amplitude_Music, Front End=$System.FrontendFakeAudio)
      B=0.05
     )
     True Float=1
     False Float=0
    )
    SpawnRate=&Multiply_Float(A=6, B=&ClampFloat(Float=&OneMinusFloat(Float=$User.HideFX)))
    Use Spawn Probability=true
   )
  ]
  particleSpawnScript [
   &PartitionParticles(Number Of Partitions=4)
   &InitializeParticle(
    Lifetime Min=0.6
    Lifetime Mode=::NewEnumerator1
    Mesh Scale=&SelectVectorFromArray(
     Array Sampling Mode=::NewEnumerator0
     Direct Array Index=?
     Vector Selection Array=DataInterface() {propertyValues {FloatData [<0.5 0.5 1>, <1 1 2>, <0.7 0.7 1>, <0.6 0.6 1.2>]}}
     Random Seed=0
    )
    Mesh Scale Mode=::NewEnumerator4
    UsePositionOffset=false
    Write Material Random=true
   )
   &InitialMeshOrientation(Mesh Orientation Mode=::NewEnumerator2, Use Rotation Vector=true)
   &CameraOffset(Camera Offset Amount=10)
  ]
  particleUpdateScript [
   &ParticleState()
   &ScaleMeshSize(Scale Factor=&VectorFromFloat(Value=&SelectFloatFromArray(Direct Array Index=$Particles.PartitionParticles.Partition, Float Selection Array=DataInterface() {propertyValues {FloatData [1,1,1,1]}}))) {module=ScaleMeshSize002}
   &SocketLocation(
    SamplePosOffset=true
    Sampled Position Offset=&Multiply_VectorByFloat(
     Vector=/Niagara/DynamicInputs/Math/Clamp_Vector(
      Vector:Vector3f=&RandomRangeVector(Minimum=<-1 -1 -1>, Maximum=<1 1 2>)
      Min:Vector3f=<-1 -1 -0.2>
      Max:Vector3f=<1 1 1>
     )
     Float=&SelectFloatFromArray(Direct Array Index=$Particles.PartitionParticles.Partition, Float Selection Array=DataInterface() {propertyValues {FloatData [5,5,7,7]}})
    )
    Socket Index=&SelectIntFromArray(Direct Array Index=$Particles.PartitionParticles.Partition, Int Selection Array=DataInterface() {propertyValues {IntData [4,5,6,7]}})
    Socket Reader=$System.Socket Reader
    Socket Sampling Mode=::NewEnumerator3
    VelocitySpeedLimit=false
   )
   &ScaleMeshSize(Scale Factor=&VectorFromFloat(Value=&MakeCustomFloatFromBool(
      Bool=$User.bisFootHidden
      True Float=&SelectFloatFromArray(Direct Array Index=$Particles.PartitionParticles.Partition, Float Selection Array=DataInterface() {propertyValues {FloatData [1,1,0,0]}})
      False Float=1
     )))
   &DynamicMaterialParameters(
    Index 0 Param 1=&RandomRangeFloat(Minimum=0.5, Maximum=1)
    Index 0 Param 2=&RandomRangeFloat(Minimum=0.8, Maximum=1)
    Index 0 Param 3=&RandomRangeFloat(Minimum=0.5, Maximum=1)
    Index 0 Param 4=&RandomRangeFloat(Minimum=0.5, Maximum=1)
    Index 3 Param 1=&SelectFloatFromArray(Direct Array Index=$Particles.PartitionParticles.Partition, Float Selection Array=DataInterface() {propertyValues {FloatData [8,8,10,10]}})
    Index 3 Param 2=&SelectFloatFromArray(Direct Array Index=$Particles.PartitionParticles.Partition, Float Selection Array=DataInterface() {propertyValues {FloatData [3,3,6,6]}})
    Index 3 Param 3=&RandomRangeFloat(Minimum=2, Maximum=2)
    Index 3 Param 4=&RandomRangeFloat(Minimum=0.5, Maximum=1)
    Module.Index0Float0Write=true
    Module.Index0Float1Write=true
    Module.Index0Float2Write=true
    Module.Index0Float3Write=true
    Module.Index3Float0Write=true
    Module.Index3Float1Write=true
    Module.Index3Float2Write=true
    Module.Index3Float3Write=true
    Write Parameter Index 3=::NewEnumerator1
   )
   &ScaleColor(
    Scale Alpha=&FloatFromCurve(
     FloatCurve=DataInterface() {propertyValues {Curve {keys [
         <at 0 0 RCIM_Cubic>
         {arriveTangent=0.0224, interpMode=RCIM_Cubic, leaveTangent=0.0224, tangentMode=RCTM_User, time=0.3, value=1}
         {arriveTangent=-0.5254, interpMode=RCIM_Cubic, leaveTangent=-0.5254, tangentMode=RCTM_User, time=0.6, value=1}
         <at 1 0 RCIM_Cubic>
        ]}}}
     Scale Curve=0.5
    )
    ScaleA=true
    ScaleRGB=true
   )
  ]
  rendererClasses [&NiagaraMeshRendererProperties]
  renderers [{rendererClass=&NiagaraMeshRendererProperties, rendererIndex=0}]
 }
 BurningParticleDecal {
  emitterUpdateScript [
   &EmitterState(Life Cycle Mode=::NewEnumerator1, Loop Behavior=::NewEnumerator1, Loop Duration Mode=::NewEnumerator1, UseLoopDelay=false)
   &SpawnBurst_Instantaneous(
    Spawn Count=&Multiply_Int(
     A=6
     B=&MakeIntFromFloat(Float=&ClampFloat(Float=&OneMinusFloat(Float=$User.HideFX)))
    )
    Use Spawn Probability=false
    UseLoopCountLimit=false
   )
  ]
  particleSpawnScript [
   &PartitionParticles(Number Of Partitions=6)
   &InitializeParticle(
    Color=<rgba 5 0.5493 0 0.35>
    Lifetime=2
    Sprite Size=&SelectVector2DFromArray(
     Direct Array Index=$Particles.PartitionParticles.Partition
     Vector 2D Selection Array=DataInterface() {propertyValues {FloatData [<15 25>, <15 25>, <30 40>, <30 40>, <15 26>, <15 26>]}}
    )
    Sprite Size Mode=::NewEnumerator4
    UsePositionOffset=false
    Write Material Random=true
   )
  ]
  particleUpdateScript [
   &ParticleState(Kill Particles When Lifetime Has Elapsed=false)
   &SocketLocation(
    Orientation Sampling=::NewEnumerator1
    SamplePosOffset=true
    Sampled Position Offset=&SelectVectorFromArray(
     Direct Array Index=$Particles.PartitionParticles.Partition
     Vector Selection Array=DataInterface() {propertyValues {FloatData [<0 0 -4.4>, <0 0 -4.4>, <0 -2 -9>, <0 2 -9>, <0 -1 0>, <0 1 0>]}}
    )
    Socket Index=&SelectIntFromArray(Direct Array Index=$Particles.PartitionParticles.Partition, Int Selection Array=DataInterface() {propertyValues {IntData [2,3,6,7,0,1]}})
    Socket Reader=$System.Socket Reader
    Socket Sampling Mode=::NewEnumerator3
    VelocitySpeedLimit=false
   )
   &ScaleSpriteSize(
    Scale Sprite Size Mode=::NewEnumerator0
    Uniform Scale Factor=&MakeCustomFloatFromBool(
     Bool=$User.bisFootHidden
     True Float=&SelectFloatFromArray(Direct Array Index=$Particles.PartitionParticles.Partition, Float Selection Array=DataInterface() {propertyValues {FloatData [1,1,0,0,1,1]}})
     False Float=1
    )
   )
   &ScaleSpriteSize(
    Scale Sprite Size Mode=::NewEnumerator0
    Uniform Scale Factor=&MakeCustomFloatFromBool(
     Bool=$User.bisHandHidden
     True Float=&SelectFloatFromArray(Direct Array Index=$Particles.PartitionParticles.Partition, Float Selection Array=DataInterface() {propertyValues {FloatData [0,0,1,1,1,1]}})
     False Float=1
    )
   ) {module=ScaleSpriteSize001}
   &DynamicMaterialParameters(
    Index 0 Param 1=1.4
    Index 0 Param 2=0.7
    Index 0 Param 4=1
    Module.Index0Float0Write=true
    Module.Index0Float1Write=true
    Module.Index0Float2Write=true
    Module.Index0Float3Write=true
   )
   &ScaleColor(Scale Alpha=&Lerp_Float(A=0, B=1, Alpha=$System.Timeline.NormalizedTime), ScaleA=true, ScaleRGB=true)
   &CameraOffset(Camera Offset Amount=&SelectFloatFromArray(
     Direct Array Index=$Particles.PartitionParticles.Partition
     Float Selection Array=DataInterface() {propertyValues {FloatData [6,6,15,15,8,8]}}
    ))
   /SetVariables(Particles.SpriteAlignment=&ReturnMeshOrientationAxis(Mesh Orientation Axis=::ZAxis, Mesh Orientation Quat=$Output.SocketLocation.SampledRotation)) {isSetParameters=true}
  ]
  rendererClasses [&NiagaraSpriteRendererProperties]
  renderers [{rendererClass=&NiagaraSpriteRendererProperties}]
 }
 Embers {
  emitterUpdateScript [
   &EmitterState(Life Cycle Mode=::NewEnumerator1, Loop Behavior=::NewEnumerator1, Loop Duration Mode=::NewEnumerator1, UseLoopDelay=false)
   &SpawnRate(
    Spawn Probability=&MakeCustomFloatFromBool(
     Bool=&SetBoolByFloatComparison(
      A=&FN_FrontEndFloatModifier(Gameplay=$NPC.NPC_Audio_Reactive_Parameters.Amplitude_Music, Front End=$System.FrontendFakeAudio)
      B=0.05
     )
     True Float=1
     False Float=0
    )
    SpawnRate=&Multiply_Float(A=20, B=&ClampFloat(Float=&OneMinusFloat(Float=$User.HideFX)))
    Use Spawn Probability=true
   )
   /SetVariables(
    Emitter.BorderColor:LinearColor=<rgba 0.3837 0.027 0 1>
    Emitter.ColdColor:LinearColor=<rgba 0.02 0.0051 0.003 1>
    Emitter.CoreColor:LinearColor=<rgba 2 1.1132 0 1>
   ) {isSetParameters=true}
  ]
  particleSpawnScript [
   &PartitionParticles(Number Of Partitions=&MakeCustomIntFromBool(Bool=$User.bisFootHidden, True INT=4, False INT=6))
   &InitializeParticle(
    Color=<rgba 2 0.7121 0 0.9014>
    Lifetime Max=1.5
    Lifetime Min=0.4
    Lifetime Mode=::NewEnumerator1
    Sprite Rotation Mode=::NewEnumerator0
    Sprite Size Mode=::NewEnumerator1
    Uniform Sprite Size Min=3
    UsePositionOffset=false
    Write Material Random=true
   )
   &SocketLocation(
    SamplePosOffset=true
    Sampled Position Offset=<0 0 10>
    Socket Index=&SelectIntFromArray(Direct Array Index=$Particles.PartitionParticles.Partition, Int Selection Array=DataInterface() {propertyValues {IntData [0,1,4,5,6,7]}})
    Socket Reader=$System.Socket Reader
    Socket Sampling Mode=::NewEnumerator3
    VelocitySpeedLimit=false
   )
   &ShapeLocation(MaskSpawnGroup=false, Override Seed=false, Sphere Radius=8, UseOwnerScale=true)
   &AddVelocity(Velocity=<0 -1 0>, Velocity Speed Scale=&RandomRangeFloat(Minimum=20, Maximum=100))
  ]
  particleUpdateScript [
   &ParticleState()
   &ScaleSpriteSize(Uniform Curve Sprite Scale=DataInterface() {propertyValues {Curve {
       defaultValue=0
       keys [
        {arriveTangent=9.0725, interpMode=RCIM_Cubic, leaveTangent=9.0725, tangentMode=RCTM_User, time=0, value=0}
        {arriveTangent=-2.5249, interpMode=RCIM_Cubic, leaveTangent=-2.5249, tangentMode=RCTM_User, time=0.1408, value=0.9847}
        {arriveTangent=-0.0405, interpMode=RCIM_Cubic, leaveTangent=-0.0405, tangentMode=RCTM_User, time=1, value=0}
       ]
      }}})
   &GravityForce(Gravity=<0 0 100>)
   &CurlNoiseForce(Bias Noise=false, Noise Frequency=20, Noise Strength=200, Pan Noise=true, Pan Noise Field=<0 0 -500>)
   &Drag(Use Linear Drag=true, Use Rotational Drag=false)
   &InheritSourceMovement(Applied Position Scale=<0.8 0.8 0.8>, ApplyForce=false, ApplyPosition=true, ApplyPositionFalloff=false)
   &DynamicMaterialParameters(
    Index 0 Param 1=&FloatFromCurve(FloatCurve=DataInterface() {propertyValues {Curve {
        defaultValue=1
        keys [
         {arriveTangent=-0.1719, interpMode=RCIM_Cubic, leaveTangent=-0.1719, tangentMode=RCTM_User, time=0, value=0}
         {interpMode=RCIM_Cubic, tangentMode=RCTM_User, time=0.6496, value=1.0259}
        ]
       }}})
    Index 0 Param 2=&RandomRangeFloat(Minimum=0, Maximum=1)
    Index 0 Param 3=&RandomRangeFloat(Minimum=0, Maximum=1)
    Module.Index0Float0Write=true
    Module.Index0Float1Write=true
    Module.Index0Float2Write=true
    Module.Index0Float3Write=true
   )
   &SolveForcesAndVelocity(Clamp Velocity=false, Limit Acceleration=false)
  ]
  rendererClasses [&NiagaraSpriteRendererProperties]
  renderers [{rendererClass=&NiagaraSpriteRendererProperties}]
 }
 FlamesSplineThicken {
  emitterUpdateScript [
   &EmitterState(
    Inactive Response=::NewEnumerator1
    Life Cycle Mode=::NewEnumerator1
    Loop Behavior=::NewEnumerator1
    Loop Duration Mode=::NewEnumerator1
    UseLoopDelay=false
   )
   &SpawnBurst_Instantaneous(
    Spawn Count=&Multiply_Int(
     A=12
     B=&MakeIntFromFloat(Float=&ClampFloat(Float=&OneMinusFloat(Float=$User.HideFX)))
    )
    Use Spawn Probability=false
    UseLoopCountLimit=false
   )
  ]
  particleSpawnScript [
   &PartitionParticles(Number Of Partitions=12)
   &InitializeParticle(
    Lifetime=5
    Mesh Scale=&SelectVectorFromArray(
     Direct Array Index=$Particles.PartitionParticles.Partition
     Vector Selection Array=DataInterface() {propertyValues {FloatData [
        <0.6 0.6 1.2>
        <0.6 0.6 1.2>
        <0.5 0.5 1>
        <0.5 0.5 1>
        <1 1 2>
        <1 1 1.6>
        <1 1 2>
        <1 1 2>
        <0.7 0.7 1.4>
        <0.5 0.5 1>
        <0.5 0.5 0.7>
        <0.3 0.3 1>
       ]}}
    )
    Mesh Scale Mode=::NewEnumerator4
    UsePositionOffset=false
    Write Material Random=true
   )
   &InitialMeshOrientation(
    Mesh Axis To Orient=<1 0 1>
    Mesh Orientation Mode=::NewEnumerator2
    Orientation Vector=<1 0 1>
    Rotation=&SelectVectorFromArray(
     Direct Array Index=$Particles.PartitionParticles.Partition
     Vector Selection Array=DataInterface() {propertyValues {FloatData [
        <0 0.027 0>
        <0 0.027 0>
        <0 0.036 0>
        <0 0.036 0>
        <0 0.027 0>
        <0 0.027 0>
        <0 0 0>
        <-0.027 0.0541 0>
        <0.045 -0.017 0>
        <-0.036 -0.009 0>
        <-0.02 0.045 0>
        <0 -0.0631 0>
       ]}}
    )
    Use Rotation Vector=true
   )
   &CameraOffset(Camera Offset Amount=&SelectFloatFromArray(
     Direct Array Index=$Particles.PartitionParticles.Partition
     Float Selection Array=DataInterface() {propertyValues {FloatData [8,8,8,8,8,8,8,8,6,6,6,7]}}
    ))
  ]
  particleUpdateScript [
   &ParticleState(Kill Particles When Lifetime Has Elapsed=false, Let Infinitely Lived Particles Die When Emitter Deactivates=true)
   &ScaleMeshSize(Scale Factor=&MakeVector(X=1, Y=1, Z=&Lerp_Float(A=1, B=1.1, Alpha=$System.Timeline001.NormalizedTime))) {module=ScaleMeshSize002}
   &SocketLocation(
    SamplePosOffset=true
    Sampled Position Offset=&SelectVectorFromArray(
     Direct Array Index=$Particles.PartitionParticles.Partition
     Vector Selection Array=DataInterface() {propertyValues {FloatData [
        <0 0 0>
        <0 0 -5>
        <2 0 0>
        <2 0 0>
        <3 0 0>
        <4 -2.6986 11.3703>
        <5 -7.6734 0>
        <1.1755 1.8135 -12>
        <-2 -2 3>
        <0 4.3846 4>
        <2.1874 1.367 19.9774>
        <0 0 2>
       ]}}
    )
    Socket Index=&SelectIntFromArray(
     Direct Array Index=$Particles.PartitionParticles.Partition
     Int Selection Array=DataInterface() {propertyValues {IntData [0,1,2,3,4,5,6,7,0,1,1,2]}}
    )
    Socket Reader=$System.Socket Reader
    Socket Sampling Mode=::NewEnumerator3
    VelocitySpeedLimit=false
   )
   &ScaleMeshSize(Scale Factor=&VectorFromFloat(Value=&MakeCustomFloatFromBool(
      Bool=$User.bisFootHidden
      True Float=&SelectFloatFromArray(
       Direct Array Index=$Particles.PartitionParticles.Partition
       Float Selection Array=DataInterface() {propertyValues {FloatData [1,1,1,1,1,1,0,0,1,1,1,1]}}
      )
      False Float=1
     )))
   &ScaleMeshSize(Scale Factor=&VectorFromFloat(Value=&MakeCustomFloatFromBool(
      Bool=$User.bisHandHidden
      True Float=&SelectFloatFromArray(
       Direct Array Index=$Particles.PartitionParticles.Partition
       Float Selection Array=DataInterface() {propertyValues {FloatData [1,1,0,0,1,1,1,1,1,1,1,0]}}
      )
      False Float=1
     ))) {module=ScaleMeshSize001}
   &InterpolateOverTime(
    Rate Of Change:NiagaraFloat=$System.Interpolate Rate
    Target Vector Value:NiagaraPosition=$Particles.Position
    Variable Type=::NewEnumerator2
   )
   /SetVariables(Particles.InterpolateDynamic:Vector4f=&MakeFloat4FromVectorAndFloat(
     "Vector (RGB)"=$Particles.InterpolateOverTime.InterpolateOverTime_V3_.Moving Average
     "Float (Alpha)"=&Multiply_Float(
      A=$System.Interpolate Internsity
      B=&SelectFloatFromArray(
       Direct Array Index=$Particles.PartitionParticles.Partition
       Float Selection Array=DataInterface() {propertyValues {FloatData [1,1,0.4,0.4,1,1,1,1,1,1,1,0.4]}}
      )
     )
    )) {isSetParameters=true}
   &DynamicMaterialParameters(
    Index 0 Param 1=&RandomRangeFloat(Minimum=0.5, Maximum=1)
    Index 0 Param 2=&RandomRangeFloat(Minimum=0.8, Maximum=1)
    Index 0 Param 3=&RandomRangeFloat(Minimum=0.5, Maximum=1)
    Index 0 Param 4=&RandomRangeFloat(Minimum=0.5, Maximum=1)
    Index 3 Param 1=&SelectFloatFromArray(
     Direct Array Index=$Particles.PartitionParticles.Partition
     Float Selection Array=DataInterface() {propertyValues {FloatData [10,10,8,8,10,10,10,10,10,15,10,10]}}
    )
    Index 3 Param 2=&SelectFloatFromArray(
     Direct Array Index=$Particles.PartitionParticles.Partition
     Float Selection Array=DataInterface() {propertyValues {FloatData [10,10,6,6,15,10,10,10,3,4,2,2]}}
    )
    Index 3 Param 3=&RandomRangeFloat(Minimum=5, Maximum=5)
    Index 3 Param 4=&RandomRangeFloat(Minimum=0.5, Maximum=1)
    Module.Index0Float0Write=true
    Module.Index0Float1Write=true
    Module.Index0Float2Write=true
    Module.Index0Float3Write=true
    Module.Index3Float0Write=true
    Module.Index3Float1Write=true
    Module.Index3Float2Write=true
    Module.Index3Float3Write=true
    Write Parameter Index 3=::NewEnumerator1
   )
   &ScaleColor(
    Scale Alpha=&Lerp_Float(A=0, B=1, Alpha=$System.Timeline.NormalizedTime)
    Scale RGB=&VectorFromFloat(Value=&Lerp_Float(A=1, B=1.2, Alpha=$System.IncomingAudio))
    ScaleA=true
    ScaleRGB=true
   )
   &CameraOffset(Camera Offset Amount=&Multiply_Float(A=$Particles.Initial.CameraOffset, B=&Lerp_Float(A=0, B=1, Alpha=$System.Timeline.NormalizedTime))) {module=CameraOffset001}
  ]
  rendererClasses [&NiagaraMeshRendererProperties]
  renderers [
   {rendererClass=&NiagaraMeshRendererProperties, rendererIndex=0}
   {rendererClass=&NiagaraMeshRendererProperties, rendererIndex=1}
   {rendererClass=&NiagaraMeshRendererProperties, rendererIndex=2}
  ]
 }
}
summary {
 systemName=NS_Character_NoiseClueGust_FX
 emitters [
  {emitterName=Embers, rendererClasses [&NiagaraSpriteRendererProperties]}
  {emitterName=FlamesSplineThicken, rendererClasses [&NiagaraMeshRendererProperties]}
  {emitterName=AdditionalFlames, rendererClasses [&NiagaraMeshRendererProperties]}
  {emitterName=BurningParticleDecal, rendererClasses [&NiagaraSpriteRendererProperties]}
 ]
 userVariables [
  User.bisFootHidden {type=NiagaraBool}
  User.bIsFrontEnd {type=NiagaraBool}
  User.bIsFrontEndPreview {type=NiagaraBool}
  User.bisHandHidden {type=NiagaraBool}
  User.DebugReactivity {type=NiagaraBool}
  User.EnableFrontendFakeAudioLoop {type=NiagaraFloat, defaultValue=0, description=""}
  User.HideFX {type=NiagaraFloat, defaultValue=0, description=""}
 ]
}
