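# Physics asset for the M_MED_NoiseClueGust male medium player body cosmetic, defining a short dynamic chain beside the right leg.
# Three dyn_chain bodies, each built from two angled capsules, link in series to a tiny MSTR_chain_01 anchor body.
# Their constraints set narrow swing1, wider swing2 and modest twist limits, with swing1 tightest at the last link.
# Unconstrained capsules on pelvis, right thigh and right calf appear to serve as colliders for the chain.
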PhysicsAsset /BRCosmetics/Characters/Player/Male/Medium/Bodies/M_MED_NoiseClueGust/Meshes/M_MED_NoiseClueGust_Physics
bodies {
 MSTR_chain_01 [{shapeType=Capsule, length=0.505, radius=0.505, shapeName=MSTR_chain_01_capsule, rotation=<rot 0 -90 0>}]
 calf_r [{
   shapeType=Capsule
   length=37.83
   radius=10.9881
   shapeName=calf_r_capsule
   center=<23.5911 2.8248 -2.2927>
   rotation=<rot -0.439 -82.6715 -89.1744>
  }]
 dyn_chain_01 [
  {
   shapeType=Capsule
   length=14.0816
   radius=5.4631
   shapeName=dyn_chain_01_capsule_BK
   center=<4.2973 -7.3921 -7.8379>
   rotation=<rot 0 0.0006 112.4967>
  }
  {
   shapeType=Capsule
   length=10.4788
   radius=5.4631
   shapeName=dyn_chain_01_capsule_FT
   center=<4.336 13.1074 -3.5787>
   rotation=<rot 0 0.0013 34.0997>
  }
 ]
 dyn_chain_02 [
  {
   shapeType=Capsule
   length=15.9358
   radius=5.4631
   shapeName=dyn_chain_02_capsule
   center=<4.2973 -8.5322 -6.9015>
   rotation=<rot 0 0.0013 120.8624>
  }
  {
   shapeType=Capsule
   length=18.7176
   radius=5.4631
   shapeName=dyn_chain_02_capsule
   center=<5.1281 11.8091 -0.6477>
   rotation=<rot -3.1057 -1.8738 42.4668>
  }
 ]
 dyn_chain_03 [
  {
   shapeType=Capsule
   length=17.7559
   radius=5.4631
   shapeName=dyn_chain_03_capsule
   center=<3.9427 -2.5648 -7.1463>
   rotation=<rot -0.0001 -7.8683 100.9614>
  }
  {
   shapeType=Capsule
   length=17.7559
   radius=5.4631
   shapeName=dyn_chain_03_capsule
   center=<5.0998 7.0907 -1.5043>
   rotation=<rot 4.4591 -4.3024 47.2957>
  }
 ]
 pelvis [{shapeType=Capsule, length=7.4052, radius=11.7024, center=<-2.3531 -1.4809 0>}]
 thigh_r [{shapeType=Capsule, length=40.1532, radius=11.3784, center=<22.7581 -0.2492 -2.753>, rotation=<rot 88.6036 -177.6945 179.9994>}]
}
constraints [
 {bone1Name=dyn_chain_01, bone2Name=MSTR_chain_01, swing1LimitDegrees=10, swing2LimitDegrees=45, twistLimitDegrees=15}
 {bone1Name=dyn_chain_02, bone2Name=dyn_chain_01, swing1LimitDegrees=10, swing2LimitDegrees=45, twistLimitDegrees=15}
 {bone1Name=dyn_chain_03, bone2Name=dyn_chain_02, swing1LimitDegrees=5, swing2LimitDegrees=45, twistLimitDegrees=15}
]
