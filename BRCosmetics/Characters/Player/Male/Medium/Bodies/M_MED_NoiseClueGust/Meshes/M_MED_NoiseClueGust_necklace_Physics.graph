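# Physics asset for the necklace mesh of the NoiseClueGust male medium body cosmetic, setting up its dyn_necklace bones for physical motion.
# Capsule bodies form a branching chain from MSTR_necklace through dyn_necklace_01, which splits into a short 02 branch and a longer 02_B to 03_B branch.
# Each joint carries the same swing and twist limit settings, with a zero second-swing limit.
# Unconstrained capsules on the upper spine, clavicles and pectoral bones provide torso collision shapes.
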PhysicsAsset /BRCosmetics/Characters/Player/Male/Medium/Bodies/M_MED_NoiseClueGust/Meshes/M_MED_NoiseClueGust_necklace_Physics
bodies {
 MSTR_necklace [{shapeType=Capsule, length=0.505, radius=0.505, shapeName=MSTR_necklace_capsule, rotation=<rot 0 -90 0>}]
 clavicle_l [
  {shapeType=Capsule, length=6.6749, radius=8.1065, shapeName=clavicle_l_capsule, center=<10.1172 -0.1836 -3.0818>, rotation=<rot 90 0 0>}
  {shapeType=Capsule, length=6.2249, radius=8.5566, shapeName=clavicle_l_capsule, center=<10.4923 2.7694 -8.3445>, rotation=<rot 90 0 0>}
 ]
 clavicle_r [
  {shapeType=Capsule, length=6.6749, radius=8.1065, shapeName=clavicle_r_capsule, center=<-10.1172 0.1836 3.0818>, rotation=<rot -90 0 0>}
  {shapeType=Capsule, length=6.2249, radius=8.5566, shapeName=clavicle_r_capsule, center=<-10.1848 -3.3945 9.3131>, rotation=<rot -90 0 0>}
 ]
 deform_pec_aim_l [{
   shapeType=Capsule
   length=9.7298
   radius=6.6283
   shapeName=deform_pec_aim_l_capsule
   center=<10.2599 -4.0951 0.454>
   rotation=<rot 53.5009 174.4865 165.1693>
  }]
 deform_pec_aim_r [{
   shapeType=Capsule
   length=9.7298
   radius=6.6283
   shapeName=deform_pec_aim_r_capsule
   center=<-10.191 4.0294 -0.4112>
   rotation=<rot -53.501 -5.5134 14.8307>
  }]
 dyn_necklace_01 [{shapeType=Capsule, length=14.1818, radius=5.5046, shapeName=dyn_necklace_01_capsule, center=<2.8378 4.5267 0>}]
 dyn_necklace_02 [{shapeType=Capsule, length=14.7935, radius=4.8929, shapeName=dyn_necklace_02_capsule, center=<3.7169 4.5058 0>}]
 dyn_necklace_02_B [{shapeType=Capsule, length=17.7644, radius=4.9282, shapeName=dyn_necklace_02_B_capsule, center=<3.6993 4.3069 0>}]
 dyn_necklace_03_B [{shapeType=Capsule, length=14.336, radius=5.3504, shapeName=dyn_necklace_03_B_capsule, center=<5.3419 5.4731 0>}]
 spine_04 [
  {shapeType=Capsule, length=16.2026, radius=13.6529, shapeName=spine_04_capsule, center=<-0.098 -1.0516 0>}
  {shapeType=Capsule, length=18.583, radius=11.2725, shapeName=spine_04_capsule, center=<3.7496 0.5548 0>}
 ]
 spine_05 [
  {shapeType=Capsule, length=16.5415, radius=15.1575, shapeName=spine_05_capsule, center=<-7.0374 -3.2807 0>, rotation=<rot 0 -90 0>}
  {shapeType=Capsule, length=20.9066, radius=10.7924, shapeName=spine_05_capsule, center=<-2.402 -1.7377 0>, rotation=<rot 0 -90 0>}
 ]
}
constraints [
 {bone1Name=dyn_necklace_01, bone2Name=MSTR_necklace, swing1LimitDegrees=30, swing2LimitDegrees=0, twistLimitDegrees=25}
 {bone1Name=dyn_necklace_02, bone2Name=dyn_necklace_01, swing1LimitDegrees=30, swing2LimitDegrees=0, twistLimitDegrees=25}
 {bone1Name=dyn_necklace_02_B, bone2Name=dyn_necklace_01, swing1LimitDegrees=30, swing2LimitDegrees=0, twistLimitDegrees=25}
 {bone1Name=dyn_necklace_03_B, bone2Name=dyn_necklace_02_B, swing1LimitDegrees=30, swing2LimitDegrees=0, twistLimitDegrees=25}
]
