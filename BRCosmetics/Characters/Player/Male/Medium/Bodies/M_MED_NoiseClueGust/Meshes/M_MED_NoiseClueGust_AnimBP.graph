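# Animation Blueprint for the NoiseClueGust player body cosmetic, adding secondary physics motion to a pose copied from the attached parent mesh.
# The copied pose runs through a Control Rig, then two RigidBody simulations with separate body and necklace physics assets, then Anim Dynamics chains for both shoes.
# Each update, a nativized gravity state function uses a parameter asset to set gravity and RigidBody settings variables, which the AnimGraph does not visibly read.
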AnimBlueprint /BRCosmetics/Characters/Player/Male/Medium/Bodies/M_MED_NoiseClueGust/Meshes/M_MED_NoiseClueGust_AnimBP
target_skeleton=&/M_MED_NoiseClueGust_Skeleton
graphs [
 AnimGraph {nodes [
   n1 = Control Rig(Source=Copy Pose From Mesh() {bUseAttachedParent=true}) {controlRigAssetReference {controlRigAsset=&/M_MED_NoiseClueGust_CtrlRig}}
   n2 = RigidBody!(
    ComponentPose=Local To Component(LocalPose=@n1)
    Alpha
    OverrideWorldGravity:Vector="0,0,0"
    ExternalForce
    SimSpaceSettings:Sim Space Settings Structure=(WorldAlpha=1,VelocityScaleZ=0.4,DampingAlpha=1,MaxLinearVelocity=300,MaxAngularVelocity=10,MaxLinearAcceleration=1000,MaxAngularAcceleration=40,ExternalLinearDragV=(X=0.3,Y=0.3,Z=0.3),ExternalLinearVelocity=(X=0,Y=0,Z=0),ExternalAngularVelocity=(X=0,Y=0,Z=0))
   ) -> (Pose) {
    bClampLinearTranslationLimitToRefPose=true
    bOverrideWorldGravity=true
    cachedBoundsScale=1
    overridePhysicsAsset=&/M_MED_NoiseClueGust_Physics
    simulationSpace=BaseBoneSpace
    alphaBoolBlend {blendOption=Linear}
    baseBoneRef {boneName=root}
    componentAppliedLinearAccClamp=<0 0 0>
    overrideWorldGravity {z=-980}
    simSpaceSettings {
     maxAngularAcceleration=40
     maxAngularVelocity=10
     maxLinearAcceleration=1000
     maxLinearVelocity=300
     velocityScaleZ=0.4
     worldAlpha=1
     externalLinearDragV=<0.3 0.3 0.3>
    }
   }
   n3 = RigidBody(
    ComponentPose=@n2
    OverrideWorldGravity="0,0,0"
    SimSpaceSettings=(WorldAlpha=1,VelocityScaleZ=0.4,DampingAlpha=1,MaxLinearVelocity=300,MaxAngularVelocity=10,MaxLinearAcceleration=1000,MaxAngularAcceleration=40,ExternalLinearDragV=(X=0.3,Y=0.3,Z=0.3),ExternalLinearVelocity=(X=0,Y=0,Z=0),ExternalAngularVelocity=(X=0,Y=0,Z=0))
   ) {
    bClampLinearTranslationLimitToRefPose=true
    bOverrideWorldGravity=true
    cachedBoundsScale=1
    overridePhysicsAsset=&/M_MED_NoiseClueGust_necklace_Physics
    simulationSpace=BaseBoneSpace
    alphaBoolBlend {blendOption=Linear}
    baseBoneRef {boneName=spine_05}
    componentAppliedLinearAccClamp=<0 0 0>
    overrideWorldGravity {z=-980}
    simSpaceSettings {
     maxAngularAcceleration=40
     maxAngularVelocity=10
     maxLinearAcceleration=1000
     maxLinearVelocity=300
     velocityScaleZ=0.4
     worldAlpha=1
     externalLinearDragV=<0.3 0.3 0.3>
    }
   }
   n4 = "Anim Dynamics - Chain: dyn_shoe_01_l -> dyn_shoe_02_l"!(ComponentPose=@n3, Alpha=1, GravityOverride:Vector="0,0,-980", ExternalForce="0,0,0") -> (Pose) {
    angularDampingOverride=0.999
    angularSpringConstant=50
    bChain=true
    bUseGravityOverride=true
    gravityScale=0.35
    simulationSpace=BoneRelative
    boundBone {boneName=dyn_shoe_01_l}
    chainEnd {boneName=dyn_shoe_02_l}
    componentAppliedLinearAccClamp=<1000 1000 1000>
    componentLinearAccScale=<0.1 0.1 0.1>
    componentLinearVelScale=<0.1 0.1 0.1>
    gravityOverride {z=-980}
    physicsBodyDefinitions [
     {
      boundBone {boneName=dyn_shoe_01_l}
      boxExtents=<20 30 30>
      constraintSetup {angularLimitsMax=<0 0 10>, angularLimitsMin=<0 0 0>, angularTarget=<1 0 0>}
      localJointOffset=<2 4 0>
     }
     {
      boundBone {boneName=dyn_shoe_02_l}
      boxExtents=<20 30 30>
      constraintSetup {angularLimitsMax=<0 0 10>, angularLimitsMin=<0 0 0>, angularTarget=<1 0 0>}
      localJointOffset=<2 4 0>
     }
    ]
    relativeSpaceBone {boneName=root}
    simSpaceSettings {maxAngularAcceleration=35, maxAngularVelocity=7, simSpaceAngularAlpha=1}
   }
   n5 = "Anim Dynamics - Chain: dyn_shoe_01_r -> dyn_shoe_02_r"!(ComponentPose=@n4, Alpha=1, GravityOverride:Vector="0,0,-980", ExternalForce="0,0,0") -> (Pose) {
    angularDampingOverride=0.999
    angularSpringConstant=50
    bChain=true
    bUseGravityOverride=true
    gravityScale=0.35
    simulationSpace=BoneRelative
    boundBone {boneName=dyn_shoe_01_r}
    chainEnd {boneName=dyn_shoe_02_r}
    componentAppliedLinearAccClamp=<1000 1000 1000>
    componentLinearAccScale=<0.1 0.1 0.1>
    componentLinearVelScale=<0.2 0.2 0.2>
    gravityOverride {z=-980}
    physicsBodyDefinitions [
     {
      boundBone {boneName=dyn_shoe_01_r}
      boxExtents=<20 30 30>
      constraintSetup {angularLimitsMax=<0 0 10>, angularLimitsMin=<0 0 0>, angularTarget=<-1 0 0>}
      localJointOffset=<-2 -4 0>
     }
     {
      boundBone {boneName=dyn_shoe_02_r}
      boxExtents=<20 30 30>
      constraintSetup {angularLimitsMax=<0 0 10>, angularLimitsMin=<0 0 0>, angularTarget=<-1 0 0>}
      localJointOffset=<-2 -4 0>
     }
    ]
    relativeSpaceBone {boneName=root}
    simSpaceSettings {maxAngularAcceleration=35, maxAngularVelocity=7, simSpaceAngularAlpha=1}
   }
   Output Pose(Result=Component To Local(ComponentPose=@n5)) {layerGroup=None}
  ]}
 EventGraph {nodes [
   Event BlueprintUpdateAnimation()
   > n1 = "Gravity Override State Function (Parameter Asset) (Nativized)"(
    CharacterPartAnimInstance=Self-Reference()
    GameTimeInSeconds=GetGameTimeInSeconds()
    Parameters=&/M_MED_NoiseClueGust_Parameters
   )
   > Set Gravity Override(Gravity Override=@n1)
   > Set RBAN Settings Out(RBAN Settings Out=@n1.OutRigidBodyAnimNodeParameters)
  ]}
]
variables [
 Gravity Override {type=Vector}
 RBAN Settings Out {
  type=FortRigidBodyAnimNodeParameters
  default {simSpaceSettings {dampingAlpha=0, maxAngularAcceleration=16, maxAngularVelocity=4, maxLinearAcceleration=800, maxLinearVelocity=200, velocityScaleZ=0.5}}
 }
]
