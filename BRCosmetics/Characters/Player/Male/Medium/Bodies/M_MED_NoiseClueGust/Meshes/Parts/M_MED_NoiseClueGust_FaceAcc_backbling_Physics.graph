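# Physics asset for the NoiseClueGust male character's FaceAcc backbling mesh part, simulating its hair.
# Capsule chains of dyn_hair bones in groups A to I hang from MSTR_hair anchor bodies, alongside head, neck, upper spine, clavicle and upper-arm collision capsules.
# Swing and twist limits vary by chain, with G, H and I tightest.
# Angularly free cross-constraints tie H_03 to both G_03 bodies, linking those three chains laterally.
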PhysicsAsset /BRCosmetics/Characters/Player/Male/Medium/Bodies/M_MED_NoiseClueGust/Meshes/Parts/M_MED_NoiseClueGust_FaceAcc_backbling_Physics
bodies {
 MSTR_hair_A [{shapeType=Capsule, length=5, radius=3, shapeName=MSTR_hair_A_capsule, rotation=<rot 0 0 90>}]
 MSTR_hair_B_01 [{shapeType=Capsule, length=5, radius=3, shapeName=MSTR_hair_B_01_capsule, rotation=<rot 0 0 90>}]
 MSTR_hair_C [{shapeType=Capsule, length=5, radius=3, shapeName=MSTR_hair_C_capsule, rotation=<rot 0 0 90>}]
 MSTR_hair_D_01 [{shapeType=Capsule, length=5, radius=3, shapeName=MSTR_hair_D_01_capsule, rotation=<rot 0 0 90>}]
 MSTR_hair_E_01 [{shapeType=Capsule, length=5, radius=3, shapeName=MSTR_hair_E_01_capsule, rotation=<rot 0 0 90>}]
 MSTR_hair_F_01_l [{shapeType=Capsule, length=2.1531, radius=3, shapeName=MSTR_hair_F_01_l_capsule, center=<0 0.5739 0.0283>, rotation=<rot 0.0001 0 -2.8205>}]
 MSTR_hair_F_01_r [{shapeType=Capsule, length=5, radius=3, shapeName=MSTR_hair_F_01_r_capsule, rotation=<rot 0 0 90>}]
 MSTR_hair_G_01_l [{shapeType=Capsule, length=5, radius=3, shapeName=MSTR_hair_G_01_l_capsule, rotation=<rot 0 0 90>}]
 MSTR_hair_G_01_r [{shapeType=Capsule, length=5, radius=3, shapeName=MSTR_hair_G_01_r_capsule, rotation=<rot 0 0 90>}]
 MSTR_hair_H_01 [{shapeType=Capsule, length=5, radius=3, shapeName=MSTR_hair_H_01_capsule, rotation=<rot 0 0 90>}]
 MSTR_hair_I_01_l [{shapeType=Capsule, length=0.505, radius=0.505, shapeName=MSTR_hair_I_01_l_capsule, rotation=<rot 0 -90 0>}]
 MSTR_hair_I_01_r [{shapeType=Capsule, length=0.505, radius=0.505, shapeName=MSTR_hair_I_01_r_capsule, rotation=<rot 0 -90 0>}]
 clavicle_l [
  {shapeType=Capsule, length=10.6816, radius=7.682, shapeName=clavicle_l_capsule, center=<8.2422 -1.4161 -1.4086>, rotation=<rot 90 0 0>}
  {shapeType=Capsule, length=5.6014, radius=7.6782, shapeName=clavicle_r_capsule, center=<6.1071 3.0645 -3.9843>, rotation=<rot -90 0 0>}
 ]
 clavicle_r [
  {shapeType=Capsule, length=12.6417, radius=7.6782, shapeName=clavicle_r_capsule, center=<-7.8776 0.6893 1.4411>, rotation=<rot -90 0 0>}
  {shapeType=Capsule, length=3.9424, radius=9.0189, shapeName=clavicle_r_capsule, center=<-7.9632 -0.8643 4.3324>, rotation=<rot -90 0 0>}
 ]
 dyn_hair_A [{shapeType=Capsule, length=6.3902, radius=1.6098, shapeName=dyn_hair_A_capsule, center=<2.6868 0 0.2475>, rotation=<rot 0 0 90>}]
 dyn_hair_B_01 [{shapeType=Capsule, length=5.7122, radius=2.2878, shapeName=dyn_hair_B_01_capsule, center=<3.472 0.2893 2.3764>, rotation=<rot 0 0 90>}]
 dyn_hair_B_02 [{
   shapeType=Capsule
   length=5.7122
   radius=2.2878
   shapeName=dyn_hair_B_02_capsule
   center=<3.3559 -0.3423 2.4216>
   rotation=<rot -2.5844 0.0001 81.9627>
  }]
 dyn_hair_C [{shapeType=Capsule, length=5, radius=3, shapeName=dyn_hair_C_capsule, center=<2.4776 0 0>, rotation=<rot 0 0 90>}]
 dyn_hair_D_01 [{shapeType=Capsule, length=5.7636, radius=2.2364, shapeName=dyn_hair_D_01_capsule, center=<2.2088 0 2.1957>, rotation=<rot 0 0 90>}]
 dyn_hair_D_02 [{shapeType=Capsule, length=5.7636, radius=2.2364, shapeName=dyn_hair_D_02_capsule, center=<2.2088 0 2.2785>, rotation=<rot 0 0 90>}]
 dyn_hair_E_01 [{shapeType=Capsule, length=5.5224, radius=2.4776, shapeName=dyn_hair_E_01_capsule, center=<3.2466 0 1.8009>, rotation=<rot 0 0 90>}]
 dyn_hair_E_02 [{shapeType=Capsule, length=5.5224, radius=2.4776, shapeName=dyn_hair_E_02_capsule, center=<3.2466 0 0.4161>, rotation=<rot 0 0 90>}]
 dyn_hair_F_01_l [{
   shapeType=Capsule
   length=1.4748
   radius=3.1774
   shapeName=dyn_hair_F_01_l_capsule
   center=<4.0319 -0.214 -0.2768>
   rotation=<rot 0.0002 0 -0.33>
  }]
 dyn_hair_F_01_r [{
   shapeType=Capsule
   length=4.3031
   radius=3.038
   shapeName=dyn_hair_F_01_r_capsule
   center=<-5.3462 0.2483 -0.2613>
   rotation=<rot -13.8734 -0.5613 30.9813>
  }]
 dyn_hair_F_02_l [{shapeType=Capsule, length=2.6986, radius=2.9361, shapeName=dyn_hair_F_02_l_capsule, center=<4.8918 -0.4108 1.6679>, rotation=<rot 0 0 90>}]
 dyn_hair_F_02_r [{
   shapeType=Capsule
   length=4.6064
   radius=3.1126
   shapeName=dyn_hair_F_02_r_capsule
   center=<-5.0885 0.6115 -1.9964>
   rotation=<rot -0.3116 -7.7915 78.4124>
  }]
 dyn_hair_F_03_l [
  {
   shapeType=Capsule
   length=3.6073
   radius=2.9956
   shapeName=dyn_hair_F_03_l_capsule
   center=<3.4892 -1.5007 3.6807>
   rotation=<rot -1.4944 -6.7837 74.6044>
  }
  {
   shapeType=Capsule
   length=0.2688
   radius=3.1137
   shapeName=dyn_hair_F_03_l_capsule
   center=<9.5251 -2.2066 3.0456>
   rotation=<rot -1.4946 -6.7836 77.4534>
  }
 ]
 dyn_hair_F_03_r [
  {
   shapeType=Capsule
   length=6.0612
   radius=3.8535
   shapeName=dyn_hair_F_03_r_capsule
   center=<-3.9378 1.6065 -4.5126>
   rotation=<rot -2.0459 -11.4285 68.1939>
  }
  {
   shapeType=Capsule
   length=5.7491
   radius=4.2311
   shapeName=dyn_hair_F_03_r_capsule
   center=<-9.6393 1.895 -3.7244>
   rotation=<rot -2.0449 -11.4286 77.2641>
  }
 ]
 dyn_hair_G_01_l [{shapeType=Capsule, length=2.7968, radius=2.4552, shapeName=dyn_hair_G_01_l_capsule, center=<3.2466 0 0.1445>, rotation=<rot 0 0 90>}]
 dyn_hair_G_01_r [{shapeType=Capsule, length=2.7346, radius=2.615, shapeName=dyn_hair_G_01_r_capsule, center=<-4.268 0 0>, rotation=<rot 0 0 90>}]
 dyn_hair_G_02_l [{
   shapeType=Capsule
   length=5.5993
   radius=3.0701
   shapeName=dyn_hair_G_02_l_capsule
   center=<3.434 -0.0692 0.6055>
   rotation=<rot -0.0001 0 79.4999>
  }]
 dyn_hair_G_02_r [{
   shapeType=Capsule
   length=6.6222
   radius=2.6023
   shapeName=dyn_hair_G_02_r_capsule
   center=<-4.0034 0.0736 -0.5097>
   rotation=<rot -0.0001 0 80.9833>
  }]
 dyn_hair_G_03_l [{
   shapeType=Capsule
   length=6.0032
   radius=4.0451
   shapeName=dyn_hair_G_03_l_capsule
   center=<5.3554 -1.7917 2.7981>
   rotation=<rot 23.257 78.4469 89.0382>
  }]
 dyn_hair_G_03_r [
  {
   shapeType=Capsule
   length=9.5342
   radius=2.9648
   shapeName=dyn_hair_G_03_r_capsule
   center=<-11.0015 0.4672 -0.8089>
   rotation=<rot 22.0429 84.7245 66.374>
  }
  {
   shapeType=Capsule
   length=4.9754
   radius=2.5164
   shapeName=dyn_hair_G_03_r_capsule
   center=<-2.9926 -2.2634 -1.8297>
   rotation=<rot 21.6246 93.5319 98.0474>
  }
 ]
 dyn_hair_H_01 [{shapeType=Capsule, length=7.6203, radius=2.7997, shapeName=dyn_hair_H_01_capsule, center=<4.5789 0 0>, rotation=<rot 0 0 90>}]
 dyn_hair_H_02 [{shapeType=Capsule, length=6.0852, radius=3.0865, shapeName=dyn_hair_H_02_capsule, center=<4.6889 0 -1.5195>, rotation=<rot 0 0 90>}]
 dyn_hair_H_03 [{shapeType=Capsule, length=7.7268, radius=3.703, shapeName=dyn_hair_H_03_capsule, center=<3.7683 0 1.0952>, rotation=<rot 0 0 90>}]
 dyn_hair_H_04 [{shapeType=Capsule, length=5.649, radius=3.9629, shapeName=dyn_hair_H_04_capsule, center=<5.8919 0 3.0608>, rotation=<rot 0 0 90>}]
 dyn_hair_I_01_l [{
   shapeType=Capsule
   length=5.685
   radius=2.315
   shapeName=dyn_hair_I_01_l_capsule
   center=<3.2496 -0.2448 0.9997>
   rotation=<rot -0.0003 0.0001 76.2416>
  }]
 dyn_hair_I_01_r [{
   shapeType=Capsule
   length=5.685
   radius=2.315
   shapeName=dyn_hair_I_01_r_capsule
   center=<-3.198 0.1285 -0.5248>
   rotation=<rot -0.0004 0.0001 76.2416>
  }]
 dyn_hair_I_02_l [{
   shapeType=Capsule
   length=5.233
   radius=2.767
   shapeName=dyn_hair_I_02_l_capsule
   center=<4.942 1.3441 0.3291>
   rotation=<rot -0.0004 0.0001 76.2417>
  }]
 dyn_hair_I_02_r [{
   shapeType=Capsule
   length=5.111
   radius=2.8544
   shapeName=dyn_hair_I_02_r_capsule
   center=<-5.8561 -2.0996 -0.7362>
   rotation=<rot -0.0004 0.0001 76.2417>
  }]
 dyn_hair_I_03_l [{
   shapeType=Capsule
   length=4.5849
   radius=3.4151
   shapeName=dyn_hair_I_03_l_capsule
   center=<4.1249 0.6372 2.628>
   rotation=<rot 4.2731 17.7712 74.2562>
  }]
 dyn_hair_I_03_r [{
   shapeType=Capsule
   length=3.7692
   radius=4.2308
   shapeName=dyn_hair_I_03_r_capsule
   center=<-4.5879 -0.2316 -2.4459>
   rotation=<rot 3.241 13.3748 66.2841>
  }]
 head [{shapeType=Capsule, length=3.6307, radius=7.6351, center=<2.2956 2.1707 0>, rotation=<rot 90 0 0>}]
 neck_01 [
  {shapeType=Capsule, length=0.4803, radius=5.506, shapeName=neck_01_capsule, center=<0.2515 1.5135 0>, rotation=<rot 0 -90 0>}
  {shapeType=Capsule, length=3.1201, radius=5.506, shapeName=neck_01_capsule, center=<-4.5046 3.2773 0>, rotation=<rot 0 -90 0>}
 ]
 neck_02 [{shapeType=Capsule, length=0.3849, radius=5.4073, shapeName=neck_02_capsule, center=<0 1.411 0>, rotation=<rot 0 -90 0>}]
 spine_04 [
  {shapeType=Capsule, length=11.375, radius=11.4898, shapeName=spine_04_capsule, center=<3.576 -0.4183 0>, rotation=<rot 0 -90 0>}
  {shapeType=Capsule, length=5.9353, radius=9.1702, shapeName=spine_04_capsule, center=<7.6998 2.9549 0>, rotation=<rot 0 -90 0>}
 ]
 spine_05 [
  {shapeType=Capsule, length=18.17, radius=10.6768, shapeName=spine_05_capsule, center=<-1.7095 -1.3192 0>, rotation=<rot 0 -90 0>}
  {shapeType=Capsule, length=18.9461, radius=6.6286, shapeName=spine_05_capsule, center=<5.6679 -1.6652 0>, rotation=<rot 0 -90 0>}
  {shapeType=Capsule, length=15.2732, radius=3.35, shapeName=spine_05_capsule, center=<11.5269 -1.2397 0>, rotation=<rot 0 -90 0>}
  {shapeType=Capsule, length=11.0937, radius=18.9063, shapeName=spine_05_capsule, center=<-3.9891 -38.9447 0>, rotation=<rot -90 2.0056 0>}
 ]
 upperarm_l [{shapeType=Capsule, length=18.718, radius=3.8013, shapeName=upperarm_l_capsule, center=<11.0859 0 0.4629>, rotation=<rot 90 0 0>}]
 upperarm_r [{shapeType=Capsule, length=18.5445, radius=3.8415, shapeName=upperarm_r_capsule, center=<-6.5187 0.459 0.8247>, rotation=<rot -90 0 0>}]
}
constraints [
 {bone1Name=dyn_hair_A, bone2Name=MSTR_hair_A, swing1LimitDegrees=30, swing2LimitDegrees=30, twistLimitDegrees=45, twistMotion=Locked}
 {bone1Name=dyn_hair_B_01, bone2Name=MSTR_hair_B_01, swing1LimitDegrees=5, swing2LimitDegrees=5, twistLimitDegrees=45, twistMotion=Locked}
 {bone1Name=dyn_hair_B_02, bone2Name=dyn_hair_B_01, swing1LimitDegrees=20, swing2LimitDegrees=20, twistLimitDegrees=45, twistMotion=Locked}
 {bone1Name=dyn_hair_C, bone2Name=MSTR_hair_C, swing1LimitDegrees=30, swing2LimitDegrees=30, twistLimitDegrees=45, twistMotion=Locked}
 {bone1Name=dyn_hair_D_01, bone2Name=MSTR_hair_D_01, swing1LimitDegrees=5, swing2LimitDegrees=5, twistLimitDegrees=45, twistMotion=Locked}
 {bone1Name=dyn_hair_D_02, bone2Name=dyn_hair_D_01, swing1LimitDegrees=20, swing2LimitDegrees=20, twistLimitDegrees=45, twistMotion=Locked}
 {bone1Name=dyn_hair_E_01, bone2Name=MSTR_hair_E_01, swing1LimitDegrees=25, swing2LimitDegrees=25, twistLimitDegrees=0}
 {bone1Name=dyn_hair_E_02, bone2Name=dyn_hair_E_01, swing1LimitDegrees=45, swing2LimitDegrees=45, twistLimitDegrees=0}
 {bone1Name=dyn_hair_F_01_l, bone2Name=MSTR_hair_F_01_l, swing1LimitDegrees=20, swing2LimitDegrees=25, twistLimitDegrees=20}
 {bone1Name=dyn_hair_F_02_l, bone2Name=dyn_hair_F_01_l, swing1LimitDegrees=20, swing2LimitDegrees=30, twistLimitDegrees=25}
 {bone1Name=dyn_hair_F_03_l, bone2Name=dyn_hair_F_02_l, swing1LimitDegrees=20, swing2LimitDegrees=35, twistLimitDegrees=25}
 {bone1Name=dyn_hair_F_01_r, bone2Name=MSTR_hair_F_01_r, swing1LimitDegrees=20, swing2LimitDegrees=25, twistLimitDegrees=20}
 {bone1Name=dyn_hair_F_02_r, bone2Name=dyn_hair_F_01_r, swing1LimitDegrees=20, swing2LimitDegrees=30, twistLimitDegrees=25}
 {bone1Name=dyn_hair_F_03_r, bone2Name=dyn_hair_F_02_r, swing1LimitDegrees=20, swing2LimitDegrees=35, twistLimitDegrees=25}
 {bone1Name=dyn_hair_G_01_l, bone2Name=MSTR_hair_G_01_l, swing1LimitDegrees=5, swing2LimitDegrees=5, twistLimitDegrees=5}
 {bone1Name=dyn_hair_G_02_l, bone2Name=dyn_hair_G_01_l, swing1LimitDegrees=5, swing2LimitDegrees=5, twistLimitDegrees=5}
 {bone1Name=dyn_hair_G_03_l, bone2Name=dyn_hair_G_02_l, swing1LimitDegrees=5, swing2LimitDegrees=5, twistLimitDegrees=5}
 {bone1Name=dyn_hair_G_01_r, bone2Name=MSTR_hair_G_01_r, swing1LimitDegrees=5, swing2LimitDegrees=5, twistLimitDegrees=5}
 {bone1Name=dyn_hair_G_02_r, bone2Name=dyn_hair_G_01_r, swing1LimitDegrees=5, swing2LimitDegrees=5, twistLimitDegrees=5}
 {bone1Name=dyn_hair_G_03_r, bone2Name=dyn_hair_G_02_r, swing1LimitDegrees=5, swing2LimitDegrees=5, twistLimitDegrees=5}
 {bone1Name=dyn_hair_H_01, bone2Name=MSTR_hair_H_01, swing1LimitDegrees=5, swing2LimitDegrees=5, twistLimitDegrees=5}
 {bone1Name=dyn_hair_H_02, bone2Name=dyn_hair_H_01, swing1LimitDegrees=5, swing2LimitDegrees=5, twistLimitDegrees=5}
 {bone1Name=dyn_hair_H_03, bone2Name=dyn_hair_H_02, swing1LimitDegrees=5, swing2LimitDegrees=5, twistLimitDegrees=5}
 {bone1Name=dyn_hair_H_04, bone2Name=dyn_hair_H_03, swing1LimitDegrees=5, swing2LimitDegrees=5, twistLimitDegrees=5}
 {bone1Name=dyn_hair_I_01_l, bone2Name=MSTR_hair_I_01_l, swing1LimitDegrees=5, swing2LimitDegrees=5, twistLimitDegrees=5}
 {bone1Name=dyn_hair_I_02_l, bone2Name=dyn_hair_I_01_l, swing1LimitDegrees=5, swing2LimitDegrees=5, twistLimitDegrees=5}
 {bone1Name=dyn_hair_I_03_l, bone2Name=dyn_hair_I_02_l, swing1LimitDegrees=5, swing2LimitDegrees=5, twistLimitDegrees=5}
 {bone1Name=dyn_hair_I_01_r, bone2Name=MSTR_hair_I_01_r, swing1LimitDegrees=5, swing2LimitDegrees=5, twistLimitDegrees=5}
 {bone1Name=dyn_hair_I_02_r, bone2Name=dyn_hair_I_01_r, swing1LimitDegrees=5, swing2LimitDegrees=5, twistLimitDegrees=5}
 {bone1Name=dyn_hair_I_03_r, bone2Name=dyn_hair_I_02_r, swing1LimitDegrees=5, swing2LimitDegrees=5, twistLimitDegrees=5}
 {
  bone1Name=dyn_hair_H_03
  bone2Name=dyn_hair_G_03_l
  swing1LimitDegrees=45
  swing1Motion=Free
  swing2LimitDegrees=45
  swing2Motion=Free
  twistLimitDegrees=45
  twistMotion=Free
 }
 {
  bone1Name=dyn_hair_H_03
  bone2Name=dyn_hair_G_03_r
  swing1LimitDegrees=45
  swing1Motion=Free
  swing2LimitDegrees=45
  swing2Motion=Free
  twistLimitDegrees=45
  twistMotion=Free
 }
]
